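; BP_WideSnoreFur_MorphTest is a demo Blueprint for testing morph targets on the WideSnore fur mesh of a quadruped cosmetic companion.
; A custom Morph event calls SetMorphTarget on the WideSnore_FurData skeletal mesh, using the Morph Target Name and Morph Target Value variables.
; BeginPlay, Tick, overlap and the construction script are empty, so nothing in this graph triggers the morph.
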(event Custom|Morph
  (Components|SkeletalMesh|SetMorphTarget (Variables|Default|GetWideSnore_FurData) (Variables|Default|GetMorph%20Target%20Name) (Variables|Default|GetMorph%20Target%20Value) false))

(event EventBeginPlay)

(event Collision|EventActorBeginOverlap
  (params (OtherActor :type Actor%20Object%20Reference)))

(event EventTick
  (params (DeltaSeconds :type Float%20%28single-precision%29)))

(fn ConstructionScript)
